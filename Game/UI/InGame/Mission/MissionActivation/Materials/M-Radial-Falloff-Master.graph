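# Translucent UI master material for the mission activation screen that builds a parameterized radial falloff gradient.
# UVs are scaled and offset per axis, then measured against a center point set by the scalar Distance parameter, which serves both axes.
# The inverted distance is shifted by an edge threshold, multiplied by a falloff, raised to a power and clamped.
# A Shadow-Col01 vector parameter supplies a color, red by default.
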Material /Game/UI/InGame/Mission/MissionActivation/Materials/M-Radial-Falloff-Master
graph {
 properties {
  bUsedWithStaticMesh=false
  blendMode=BLEND_Translucent
  materialDomain=MD_UI
  pixelDepthOffsetMode=PDOM_Legacy
  usedShadingModels=Default Lit
 }
 nodes [
  n1 = ScalarParameter() {defaultValue=1, parameterName=X-Scale}
  n2 = ScalarParameter() {defaultValue=1, parameterName=Y-Scale}
  n3 = Add(
   A=Multiply(A=TextureCoordinate(), B=AppendVector(A=@n1, B=@n2))
   B=AppendVector(A=ScalarParameter() {parameterName=X-Offset}, B=ScalarParameter() {parameterName=Y-Offset})
  )
  n4 = ScalarParameter() {defaultValue=0.5, parameterName=Distance, sliderMax=1}
  n5 = Subtract(A=OneMinus(None=Distance(A=@n3, B=@n4)), B=ScalarParameter() {defaultValue=0.5, parameterName=GradientEdge-Dist}) {constB=0.25}
  n6 = ScalarParameter() {defaultValue=0.5, parameterName=GradientEdge-Falloff}
  Clamp(None=Power(Base=Multiply(A=@n5, B=@n6) {constB=6}, Exp=ScalarParameter() {defaultValue=1, parameterName=PowerVal}))
  VectorParameter() {parameterName=Shadow-Col01, defaultValue=<rgba 1 0 0 1>}
 ]
}
